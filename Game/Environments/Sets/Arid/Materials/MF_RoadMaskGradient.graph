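# Library-exposed material function that builds a road mask peaking at the center of the U texture axis and fading toward both edges.
# Multiplies U and its inverse, each reduced by Mask Width Push to narrow the band, then scales by Gradient Mask Value and saturates.
# The falloff is parabolic before clamping, despite the "linear" description.
# Both inputs default to scalar parameters in the RoadMaskGradient group.
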MaterialFunction /Game/Environments/Sets/Arid/Materials/MF_RoadMaskGradient
graph {
 properties {bExposeToLibrary=true, description=Linear Gradient to the center}
 nodes [
  n1 = ComponentMask(None=TextureCoordinate()) {r=true}
  n2 = FunctionInput(Preview=ScalarParameter() {defaultValue=0.05, group=RoadMaskGradient, parameterName=Mask Width Push}) {bUsePreviewValueAsDefault=true, inputName=Mask Width Push, inputType=FunctionInput_Scalar}
  n3 = Multiply(A=Subtract(A=@n1, B=@n2), B=Subtract(A=OneMinus(None=@n1), B=@n2))
  n4 = FunctionInput(Preview=ScalarParameter() {defaultValue=8, group=RoadMaskGradient, parameterName=Gradient Mask Value}) {bUsePreviewValueAsDefault=true, inputName=Gradient Mask Value, inputType=FunctionInput_Scalar}
  FunctionOutput(None=Saturate(None=Multiply(A=@n3, B=@n4)))
 ]
}
