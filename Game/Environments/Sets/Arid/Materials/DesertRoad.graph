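# Material function that layers a painted desert asphalt road over the incoming material attributes.
# World-aligned asphalt textures get an edge-to-center tint and roughness gradient from MF_RoadMaskGradient.
# A road mask places edge and center lines, broken up by the asphalt mask, with normals blended in.
# The road mask times vertex alpha blends the layer in, height-lerped with swirly noise; mobile skips that noise and the line roughness variation.
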MaterialFunction /Game/Environments/Sets/Arid/Materials/DesertRoad
graph {nodes [
  n1 = BreakMaterialAttributes!(Attr=FunctionInput() {inputType=FunctionInput_MaterialAttributes})
  n2 = VectorParameter() {group=Road, parameterName=AsphaltTint, defaultValue=<rgba 1 1 1 0>}
  n3 = ScalarParameter() {defaultValue=500, group=Road, parameterName=AsphaltScale}
  n4 = MaterialFunctionCall(TextureObject=TextureObject() {texture=/Game/Packages/DS_Fortnight/SM/Textures/T_Street_Road3}, TextureSize=@n3) {materialFunction=&WorldAlignedTexture}
  n5 = MaterialFunctionCall(Float3=MaterialFunctionCall(TextureObject=TextureObject() {texture=/Game/Packages/DS_Fortnight/SM/Textures/T_Street_RoadUPD_M}, TextureSize=@n3) {materialFunction=&WorldAlignedTexture}.Z Texture) {materialFunction=&BreakOutFloat3Components}
  n6 = ScalarParameter() {defaultValue=0.4, group=Road, parameterName=Asphalt R Scale_Min}
  n7 = ScalarParameter() {defaultValue=1, group=Road, parameterName=Asphalt R Scale Max}
  n8 = MaterialFunctionCall(
   TextureObject=TextureObject() {samplerType=SAMPLERTYPE_Normal, texture=/Game/Packages/DS_Fortnight/SM/Textures/T_Street_Road_N}
   TextureSize=@n3
  ) {materialFunction=&WorldAlignedNormal}
  n9 = VectorParameter() {group=Road, parameterName=RoadTile N Value, defaultValue=<rgba 2 2 1 0>}
  n10 = GetMaterialAttributes(None=MakeMaterialAttributes!(
    BaseColor=Multiply(A=@n2.RGB, B=@n4.Z Texture)
    Metallic
    Specular=@n5.R
    Roughness=LinearInterpolate(A=@n6, B=@n7, Alpha=@n5.B)
    Anisotropy
    EmissiveColor
    Opacity
    OpacityMask
    Normal=Multiply(A=@n8.Z Texture, B=@n9.RGB)
    Tangent
    WorldPositionOffset
    SubsurfaceColor
    ClearCoat
    ClearCoatRoughness
    AmbientOcclusion
    Refraction
    CustomizedUVs_0
    CustomizedUVs_1
    CustomizedUVs_2
    CustomizedUVs_3
    CustomizedUVs_4
    CustomizedUVs_5
    CustomizedUVs_6
    CustomizedUVs_7
    PixelDepthOffset
    ShadingModel
    Displacement
   )) {attributeGetTypes [#1, #2]}
  n11 = MaterialFunctionCall!(Mask Width Push, Gradient Mask Value) {materialFunction=&/MF_RoadMaskGradient}
  n12 = LinearInterpolate(
   A=VectorParameter() {group=Road, parameterName=RoadTint_Edge, defaultValue=<rgba 0.6771 0.5757 0.4246 0>}.RGB
   B=VectorParameter() {group=Road, parameterName=RoadTint_Center, defaultValue=<rgba 0.1979 0.1979 0.1979 0>}.RGB
   Alpha=@n11.Result
  )
  n13 = Multiply(B=@n10.Roughness) {constA=1.5}
  n14 = GetMaterialAttributes(None=SetMaterialAttributes!(
    MaterialAttributes=@n10.MaterialAttributes
    BaseColor=Multiply(A=@n10.BaseColor, B=@n12)
    Roughness=LinearInterpolate(A=@n13, B=@n10.Roughness, Alpha=@n11.Result)
   ) {attributeSetTypes [#1, #2]}) {attributeGetTypes [
    #2
    #3
    #1
    #4
   ]}
  n15 = ScalarParameter() {defaultValue=0.5, group=Road, parameterName=Lines_R2}
  n16 = TextureObject() {texture=/Game/Environments/Sets/Arid/Textures/T_Swirly_Noise_D}
  n17 = ScalarParameter() {defaultValue=350, group=Road, parameterName=RoadTransDetailMask}
  n18 = LinearInterpolate(
   A=@n15
   B=ScalarParameter() {defaultValue=0.9575, group=Road, parameterName=Lines_R}
   Alpha=ComponentMask(None=MaterialFunctionCall(TextureObject=@n16, TextureSize=@n17) {materialFunction=&WorldAlignedTexture}.Z Texture) {b=true}
  )
  n19 = TextureSample() {samplerType=SAMPLERTYPE_Masks, texture=/Game/Environments/Sets/Arid/Textures/SM_DesertRoad_M}
  n20 = LinearInterpolate(A=@n19.G, Alpha=@n5.G) {constB=0}
  n21 = VectorParameter() {group=Road, parameterName=EdgeLine, defaultValue=<rgba 0.7344 0.6848 0.6196 0>}
  n22 = VectorParameter() {group=Road, parameterName=CenterLine, defaultValue=<rgba 0.4115 0.2336 0 0>}
  n23 = TextureSample() {samplerType=SAMPLERTYPE_Normal, texture=/Game/Environments/Sets/Arid/Textures/T_DesertRoad_N}
  n24 = MaterialFunctionCall(BaseNormal=@n14.Normal, AdditionalNormal=@n23.RGB) {materialFunction=&BlendAngleCorrectedNormals}
  n25 = SetMaterialAttributes!(
   MaterialAttributes=@n14.MaterialAttributes
   Roughness=LinearInterpolate(A=@n14.Roughness, B=ShadingPathSwitch(Default=@n18, Mobile=@n15) {desc=Migrated from FeatureLevelSwitch}, Alpha=@n20)
   Specular=LinearInterpolate(A=@n14.Specular, Alpha=@n20) {constB=0.6}
   BaseColor=LinearInterpolate(A=@n14.BaseColor, B=LinearInterpolate(A=@n21.RGB, B=@n22.RGB, Alpha=@n19.B), Alpha=@n20)
   Normal=LinearInterpolate(A=LinearInterpolate(A=@n24.Result, B=@n23.RGB, Alpha=@n20), B=@n14.Normal, Alpha=@n5.G)
  ) {attributeSetTypes [
    #2
    #3
    #1
    #4
   ]}
  n26 = TextureSample() {samplerType=SAMPLERTYPE_Masks, texture=&SM_DesertRoad_M}
  n27 = VertexColor()
  n28 = ComponentMask(None=MaterialFunctionCall(TextureObject=@n16, TextureSize=Multiply(B=@n17) {constA=2}) {materialFunction=&WorldAlignedTexture}.Z Texture) {g=true}
  n29 = LinearInterpolate(Alpha=@n28) {constB=15}
  n30 = Multiply(A=@n28) {constB=0.5}
  n31 = ShadingPathSwitch(
   Default=MaterialFunctionCall(Transition Phase=@n27.A, Height Texture=Multiply(A=@n29, B=@n30)) {materialFunction=&HeightLerp}.Alpha
   Mobile=@n27.A
  ) {desc=Migrated from FeatureLevelSwitch}
  FunctionOutput(None=BlendMaterialAttributes(
    A=MakeMaterialAttributes(
     BaseColor=@n1.BaseColor
     Metallic=@n1.BaseColor
     Specular=@n1.BaseColor
     Roughness=@n1.BaseColor
     EmissiveColor=@n1.BaseColor
     Opacity=@n1.BaseColor
     OpacityMask=@n1.BaseColor
     Normal=@n1.BaseColor
     WorldPositionOffset=@n1.BaseColor
     SubsurfaceColor=@n1.BaseColor
     ClearCoat=@n1.BaseColor
     ClearCoatRoughness=@n1.BaseColor
     AmbientOcclusion=@n1.BaseColor
     Refraction=@n1.BaseColor
     CustomizedUVs_0=@n1.BaseColor
     CustomizedUVs_1=@n1.BaseColor
     CustomizedUVs_2=@n1.BaseColor
     CustomizedUVs_3=@n1.BaseColor
     CustomizedUVs_4=@n1.BaseColor
     CustomizedUVs_5=@n1.BaseColor
     CustomizedUVs_6=@n1.BaseColor
     CustomizedUVs_7=@n1.BaseColor
     PixelDepthOffset=@n1.BaseColor
     ShadingModel=@n1.BaseColor
    )
    B=@n25
    Alpha=Multiply(A=@n26.R, B=@n31)
   ))
 ]}
