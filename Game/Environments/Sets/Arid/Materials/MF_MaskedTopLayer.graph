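# Material function that builds a masked top-layer surface for Arid environment materials, defaulting to GeoSlope textures.
# It samples base color, packed mask and normal textures on an input UV, using shared wrap samplers.
# Mask red blends two specular parameters, blue scales roughness, and green drives opacity mask.
# It outputs the combined material attributes, plus the green channel as a separate Mask output.
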MaterialFunction /Game/Environments/Sets/Arid/Materials/MF_MaskedTopLayer
graph {nodes [
  n1 = FunctionInput(Preview=TextureCoordinate()) {bUsePreviewValueAsDefault=true, inputName=UV, inputType=FunctionInput_Vector2}
  n2 = TextureSampleParameter2D(UVs=@n1) {
   group=MaskedTopLayer
   parameterName=BaseColor_MaskedTopLayer
   samplerSource=SSM_Wrap_WorldGroupSettings
   texture=/Game/Environments/Sets/Arid/Textures/T_S_GeoSlope_2W_Arid_D
  }
  n3 = ScalarParameter() {defaultValue=0.2092, group=MaskedTopLayer, parameterName=S_1_MaskedTopLayer}
  n4 = ScalarParameter() {defaultValue=0.5, group=MaskedTopLayer, parameterName=S_2_MaskedTopLayer}
  n5 = TextureSampleParameter2D(UVs=@n1) {
   group=MaskedTopLayer
   parameterName=Mask_MaskedTopLayer
   samplerSource=SSM_Wrap_WorldGroupSettings
   samplerType=SAMPLERTYPE_Masks
   texture=/Game/Environments/Sets/Arid/Textures/T_S_GeoSlope_2W_Arid_M
  }
  n6 = ScalarParameter() {defaultValue=1.04, group=MaskedTopLayer, parameterName=R_MaskedTopLayer}
  n7 = TextureSampleParameter2D(UVs=@n1) {
   group=MaskedTopLayer
   parameterName=Normal_MaskedTopLayer
   samplerSource=SSM_Wrap_WorldGroupSettings
   samplerType=SAMPLERTYPE_Normal
   texture=/Game/Environments/Sets/Arid/Textures/T_S_GeoSlope_2W_Arid_N
  }
  FunctionOutput(None=MakeMaterialAttributes!(
    BaseColor=@n2.RGB
    Metallic
    Specular=LinearInterpolate(A=@n3, B=@n4, Alpha=@n5.R)
    Roughness=Multiply(A=@n5.B, B=@n6)
    Anisotropy
    EmissiveColor
    Opacity
    OpacityMask=@n5.G
    Normal=@n7.RGB
    Tangent
    WorldPositionOffset
    SubsurfaceColor
    ClearCoat
    ClearCoatRoughness
    AmbientOcclusion
    Refraction
    CustomizedUVs_0
    CustomizedUVs_1
    CustomizedUVs_2
    CustomizedUVs_3
    CustomizedUVs_4
    CustomizedUVs_5
    CustomizedUVs_6
    CustomizedUVs_7
    PixelDepthOffset
    ShadingModel
    Displacement
   )) {sortPriority=1}
  FunctionOutput(None=@n5.G) {outputName=Mask, sortPriority=4}
 ]}
