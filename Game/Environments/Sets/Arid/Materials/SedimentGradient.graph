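# Arid environment material function that tints incoming material attributes with world-space sediment banding.
# World Z, offset by a sine wave driven by world XY, is remapped between bottom and top blend heights to sample a gradient texture.
# The sample is desaturated, brightened and multiplied by a wall tint, then blended toward a flat-top color using world vertex normal Z.
# The result feeds MatLayerBlend_Tint.
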MaterialFunction /Game/Environments/Sets/Arid/Materials/SedimentGradient
graph {nodes [
  n1 = WorldPosition()
  n2 = Divide(
   A=Add(A=ComponentMask(None=@n1.XYZ) {r=true}, B=ComponentMask(None=@n1.XYZ) {g=true, r=true})
   B=ScalarParameter() {defaultValue=760.0527, group=SedimentColoration, parameterName=WaveLength}
  )
  n3 = Add(
   A=ComponentMask(None=@n1.XYZ) {b=true}
   B=Multiply(A=ScalarParameter() {defaultValue=33.0688, group=SedimentColoration, parameterName=WaveAmplitude}, B=Sine(None=@n2))
  )
  n4 = ScalarParameter() {defaultValue=-6539.7803, group=SedimentColoration, parameterName=BlendBottom}
  n5 = ScalarParameter() {defaultValue=-8978.6338, group=SedimentColoration, parameterName=BlendTop}
  n6 = TextureSample(UVs=Divide(A=Subtract(A=@n3, B=@n4), B=Subtract(A=@n5, B=@n4))) {samplerSource=SSM_Wrap_WorldGroupSettings, texture=/Game/Environments/Sets/Arid/Textures/GradientSediments__M}
  n7 = ScalarParameter() {defaultValue=0.7431, group=SedimentColoration, parameterName=Desat}
  n8 = Multiply(
   A=VectorParameter() {group=SedimentColoration, parameterName=WallColorTint, defaultValue=<rgba 0.87 0.5862 0.4089 1>}.RGB
   B=Add(A=Desaturation(None=@n6.RGB, Fraction=@n7), B=Constant3Vector() {constant=<rgba 0.5 0.5 0.5 0>})
  )
  n9 = VectorParameter() {group=SedimentColoration, parameterName=FlatTopBlendCOlor, defaultValue=<rgba 1 1 1 1>}
  n10 = Clamp(None=ComponentMask(None=VertexNormalWS()) {b=true})
  n11 = ScalarParameter() {defaultValue=-0.7424, group=SedimentColoration, parameterName=ShiftSedimentEdge}
  n12 = Clamp(None=LinearInterpolate(
    A=ScalarParameter() {defaultValue=0.5963, group=SedimentColoration, parameterName=SedimentEdge_LOW}
    B=ScalarParameter() {defaultValue=3.2442, group=SedimentColoration, parameterName=SedimentEdge_High}
    Alpha=Add(A=@n10, B=@n11)
   ))
  FunctionOutput(None=MaterialFunctionCall!(Material=FunctionInput() {inputType=FunctionInput_MaterialAttributes}, Tint=LinearInterpolate(A=@n8, B=@n9.RGB, Alpha=@n12), TintMask) {materialFunction=/Engine/Functions/MaterialLayerFunctions/MatLayerBlend_Tint}.Blended Material)
  Multiply()
 ]}
