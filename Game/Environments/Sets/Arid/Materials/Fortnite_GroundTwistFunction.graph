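# Material function for Fortnite ground materials that distorts scaled UVs with animated swirls in a grid of cells, outputting Distortion and Rotation Amount.
# UVs pan by time and flow direction, then stagger by half a cell per column.
# A repeating per-cell value offsets a sine wave's phase and picks the twist's sign.
# Each cell rotates about its center, weighted by a radial sphere mask with adjustable falloff, hiding tiling without texture noise.
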MaterialFunction /Game/Environments/Sets/Arid/Materials/Fortnite_GroundTwistFunction
graph {
 properties {bExposeToLibrary=true, libraryCategoriesText [Misc,Fortnite]}
 nodes [
  Constant()
  n1 = FunctionInput() {bUsePreviewValueAsDefault=true, inputName=Time, inputType=FunctionInput_Scalar, previewValue=<0 1 0 1>}
  n2 = FunctionInput() {bUsePreviewValueAsDefault=true, inputName=Flow Direction, inputType=FunctionInput_Vector2, previewValue=<0 1 0 1>}
  n3 = FunctionInput() {bUsePreviewValueAsDefault=true, inputName=UVS * size, inputType=FunctionInput_Vector2, previewValue=<128 128 0 1>}
  n4 = Add(A=Multiply(A=@n1, B=@n2), B=@n3) {constA=0.5}
  n5 = ComponentMask(None=@n4) {r=true}
  n6 = Ceil(None=Add(A=@n5) {constB=0.5})
  n7 = ComponentMask(None=@n4) {g=true}
  n8 = AppendVector(A=@n5, B=Add(A=Multiply(A=@n6) {constB=0.5}, B=Multiply(A=@n7)))
  n9 = FunctionInput() {bUsePreviewValueAsDefault=true, inputName=Grid Noise Tile Rep, inputType=FunctionInput_Vector2, previewValue=<5 3 0 1>}
  n10 = Frac(None=Divide(A=Add(A=@n8) {constB=-0.5}, B=@n9) {constB=5}) {desc=mod}
  n11 = Divide(A=Floor(None=Multiply(A=@n10, B=@n9) {constB=10}), B=@n9) {desc=0-1 values that have the number of increments specified}
  n12 = Frac(None=Add(A=ComponentMask(None=@n11) {g=true}, B=ComponentMask(None=@n11) {r=true}))
  n13 = FunctionInput() {bUsePreviewValueAsDefault=true, inputName=SineAnimSpeed, inputType=FunctionInput_Scalar, previewValue=<0.3 3 0 1>}
  n14 = FunctionInput() {bUsePreviewValueAsDefault=true, inputName=SineAnimOffset, inputType=FunctionInput_Scalar, previewValue=<0 3 0 1>}
  n15 = Add(A=@n12, B=Add(A=Multiply(A=@n1, B=@n13), B=@n14)) {desc=offset wave phase}
  n16 = FunctionInput() {bUsePreviewValueAsDefault=true, inputName=Result Bias, inputType=FunctionInput_Scalar, sortPriority=-5, previewValue=<2 0 0 1>}
  n17 = FunctionInput() {bUsePreviewValueAsDefault=true, inputName=Result Modulation, inputType=FunctionInput_Scalar, sortPriority=-4, previewValue=<0.25 0 0 1>}
  n18 = Multiply(A=Add(A=Sine(None=@n15), B=@n16), B=@n17)
  FunctionOutput(None=@n18) {outputName=Rotation Amount}
  n19 = Constant3Vector() {constant=<rgba 0 0 1 0>}
  n20 = Add(A=@n8) {constB=0.5}
  n21 = SphereMask(A=Frac(None=@n20), B=Constant() {r=0.5}) {attenuationRadius=0.5, desc=23, hardnessPercent=0}
  n22 = FunctionInput() {bUsePreviewValueAsDefault=true, inputName=Twist Falloff, inputType=FunctionInput_Scalar, previewValue=<1 0 0 1>}
  n23 = Power(Base=@n21, Exp=@n22)
  n24 = LinearInterpolate(Alpha=Ceil(None=Subtract(A=@n12) {constB=0.5})) {constA=-1}
  n25 = FunctionInput() {bUsePreviewValueAsDefault=true, inputName=Rotation amount modulator, inputType=FunctionInput_Scalar, previewValue=<0.5 0 0 1>}
  n26 = Multiply(A=Multiply(A=Multiply(A=@n23, B=@n24) {desc=-1 or +1}, B=@n25) {constA=1}, B=@n18) {desc=+15 for random height}
  n27 = Constant()
  n28 = AppendVector(A=Subtract(A=Ceil(None=@n20)) {constB=0.5}, B=@n27)
  n29 = AppendVector(A=@n20, B=@n27)
  FunctionOutput(None=ComponentMask(None=RotateAboutAxis(NormalizedRotationAxis=@n19, RotationAngle=@n26, PivotPoint=@n28, Position=@n29)) {g=true, r=true}) {outputName=Distortion, sortPriority=1}
  Multiply(A=@n21, B=@n21)
 ]
}
