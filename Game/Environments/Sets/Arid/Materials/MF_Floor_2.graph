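# Material function for the second floor layer of the Arid environment set, defaulting to rock gravel textures, that outputs material attributes plus a Height value.
# Scaled UVs optionally pass through BumpOffset, driven by mask green and a custom view vector from cross products of world Z, camera vector and vertex normal.
# Those UVs sample base color, normal and packed mask: red blends specular, blue scales roughness, green is Height.
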MaterialFunction /Game/Environments/Sets/Arid/Materials/MF_Floor_2
graph {nodes [
  n1 = Multiply(
   A=FunctionInput(Preview=TextureCoordinate()) {bUsePreviewValueAsDefault=true, inputName=UV, inputType=FunctionInput_Vector2}
   B=ScalarParameter() {defaultValue=1, group=Floor_2, parameterName=Scale_Floor_2}
  )
  n2 = TextureSampleParameter2D(UVs=@n1) {
   group=Floor_2
   parameterName=Mask_Floor_2
   samplerSource=SSM_Wrap_WorldGroupSettings
   samplerType=SAMPLERTYPE_Masks
   texture=/Game/Environments/Sets/Arid/Textures/RockGravel_Mask
  }
  n3 = CrossProduct(A=VertexNormalWS(), B=CrossProduct(A=Constant3Vector() {constant=<rgba 0 0 1 0>}, B=CameraVectorWS()))
  n4 = StaticSwitchParameter(
   True=MaterialFunctionCall!(
    HeightTexture=@n2.G
    TextureCoordinates=@n1
    ReferencePlane
    HeightRatioInput=ScalarParameter() {defaultValue=0.02, group=Floor_2, parameterName=BumpOffset_Floor_2}
    CameraVector=Normalize(VectorInput=@n3)
   ) {materialFunction=/Engine/Functions/Engine_MaterialFunctions02/Utility/BumpOffset_Advanced.BumpOffset_advanced}.UVs
   False=@n1
  ) {defaultValue=true, parameterName=Use_BumpOffset}
  n5 = TextureSampleParameter2D(UVs=@n4) {
   group=Floor_2
   parameterName=BaseColor_Floor_2
   samplerSource=SSM_Wrap_WorldGroupSettings
   texture=/Game/Environments/Sets/Arid/Textures/RockGravel_basecolor
  }
  n6 = ScalarParameter() {defaultValue=0.2092, group=Floor_2, parameterName=Spec_low_Floor_2}
  n7 = ScalarParameter() {defaultValue=0.3991, group=Floor_2, parameterName=Spec_High_Floor_2}
  n8 = TextureSampleParameter2D(UVs=@n4) {
   group=Floor_2
   parameterName=Mask_Floor_2
   samplerSource=SSM_Wrap_WorldGroupSettings
   samplerType=SAMPLERTYPE_Masks
   texture=&RockGravel_Mask
  }
  n9 = ScalarParameter() {defaultValue=1, group=Floor_2, parameterName=Roughness_Floor_2}
  n10 = TextureSampleParameter2D(UVs=@n4) {
   group=Floor_2
   parameterName=Normal_Floor_2
   samplerSource=SSM_Wrap_WorldGroupSettings
   samplerType=SAMPLERTYPE_Normal
   texture=/Game/Environments/Sets/Arid/Textures/RockGravel_normal
  }
  FunctionOutput(None=MakeMaterialAttributes!(
    BaseColor=@n5.RGB
    Metallic
    Specular=LinearInterpolate(A=@n6, B=@n7, Alpha=@n8.R)
    Roughness=Multiply(A=@n8.B, B=@n9)
    Anisotropy
    EmissiveColor
    Opacity
    OpacityMask
    Normal=@n10.RGB
    Tangent
    WorldPositionOffset
    SubsurfaceColor
    ClearCoat
    ClearCoatRoughness
    AmbientOcclusion
    Refraction
    CustomizedUVs_0
    CustomizedUVs_1
    CustomizedUVs_2
    CustomizedUVs_3
    CustomizedUVs_4
    CustomizedUVs_5
    CustomizedUVs_6
    CustomizedUVs_7
    PixelDepthOffset
    ShadingModel
    Displacement
   ))
  FunctionOutput(None=@n8.G) {outputName=Height}
 ]}
