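# Material function for the Floor_3 surface layer of the Arid environment set, defaulting to rock gravel textures and outputting material attributes plus height.
# Scaled input UVs sample base color, normal and a packed mask: R blends low/high specular, G is height, B scales roughness.
# A static switch, on by default, applies bump-offset parallax from mask height, using a custom camera vector built from cross products of world normal, Z axis and view.
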MaterialFunction /Game/Environments/Sets/Arid/Materials/MF_Floor_3
graph {nodes [
  n1 = Multiply(
   A=FunctionInput(Preview=TextureCoordinate()) {bUsePreviewValueAsDefault=true, inputName=UV, inputType=FunctionInput_Vector2}
   B=ScalarParameter() {defaultValue=1, group=Floor_3, parameterName=Scale_Floor_3}
  )
  n2 = TextureSampleParameter2D(UVs=@n1) {
   group=Floor_3
   parameterName=Mask_Floor_3
   samplerSource=SSM_Wrap_WorldGroupSettings
   samplerType=SAMPLERTYPE_Masks
   texture=/Game/Environments/Sets/Arid/Textures/RockGravel_Mask
  }
  n3 = CrossProduct(A=VertexNormalWS(), B=CrossProduct(A=Constant3Vector() {constant=<rgba 0 0 1 0>}, B=CameraVectorWS()))
  n4 = StaticSwitchParameter(
   True=MaterialFunctionCall!(
    HeightTexture=@n2.G
    TextureCoordinates=@n1
    ReferencePlane
    HeightRatioInput=ScalarParameter() {defaultValue=0.02, group=Floor_3, parameterName=BumpOffset_Floor_3}
    CameraVector=Normalize(VectorInput=@n3)
   ) {materialFunction=/Engine/Functions/Engine_MaterialFunctions02/Utility/BumpOffset_Advanced.BumpOffset_advanced}.UVs
   False=@n1
  ) {defaultValue=true, parameterName=Use_BumpOffset}
  n5 = TextureSampleParameter2D(UVs=@n4) {
   group=Floor_3
   parameterName=BaseColor_Floor_3
   samplerSource=SSM_Wrap_WorldGroupSettings
   texture=/Game/Environments/Sets/Arid/Textures/RockGravel_basecolor
  }
  n6 = ScalarParameter() {defaultValue=0.2092, group=Floor_3, parameterName=Spec_low_Floor_3}
  n7 = ScalarParameter() {defaultValue=0.3991, group=Floor_3, parameterName=Spec_High_Floor_3}
  n8 = TextureSampleParameter2D(UVs=@n4) {
   group=Floor_3
   parameterName=Mask_Floor_3
   samplerSource=SSM_Wrap_WorldGroupSettings
   samplerType=SAMPLERTYPE_Masks
   texture=&RockGravel_Mask
  }
  n9 = ScalarParameter() {defaultValue=1, group=Floor_3, parameterName=Roughness_Floor_3}
  n10 = TextureSampleParameter2D(UVs=@n4) {
   group=Floor_3
   parameterName=Normal_Floor_3
   samplerSource=SSM_Wrap_WorldGroupSettings
   samplerType=SAMPLERTYPE_Normal
   texture=/Game/Environments/Sets/Arid/Textures/RockGravel_normal
  }
  FunctionOutput(None=MakeMaterialAttributes!(
    BaseColor=@n5.RGB
    Metallic
    Specular=LinearInterpolate(A=@n6, B=@n7, Alpha=@n8.R)
    Roughness=Multiply(A=@n8.B, B=@n9)
    Anisotropy
    EmissiveColor
    Opacity
    OpacityMask
    Normal=@n10.RGB
    Tangent
    WorldPositionOffset
    SubsurfaceColor
    ClearCoat
    ClearCoatRoughness
    AmbientOcclusion
    Refraction
    CustomizedUVs_0
    CustomizedUVs_1
    CustomizedUVs_2
    CustomizedUVs_3
    CustomizedUVs_4
    CustomizedUVs_5
    CustomizedUVs_6
    CustomizedUVs_7
    PixelDepthOffset
    ShadingModel
    Displacement
   ))
  FunctionOutput(None=@n8.G) {outputName=Height}
 ]}
