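# Floor material function for the Arid environment set, defaulting to desert sand textures and outputting material attributes plus a Height value.
# Scaled input UVs optionally pass through bump offset, driven by mask green and a custom view vector built from cross products of world up, camera vector and vertex normal.
# The resulting UVs sample base color, normal and mask; mask red blends specular, blue scales roughness and green supplies Height.
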MaterialFunction /Game/Environments/Sets/Arid/Materials/ML_Floor_1
graph {nodes [
  n1 = Multiply(
   A=FunctionInput(Preview=TextureCoordinate()) {bUsePreviewValueAsDefault=true, inputName=UV, inputType=FunctionInput_Vector2}
   B=ScalarParameter() {defaultValue=1, group=Floor_1, parameterName=Scale_Floor_1}
  )
  n2 = TextureSampleParameter2D(UVs=@n1) {
   group=Floor_1
   parameterName=Mask_Floor_1
   samplerSource=SSM_Wrap_WorldGroupSettings
   samplerType=SAMPLERTYPE_Masks
   texture=/Game/Environments/Sets/Arid/Textures/DesertSand_Mask
  }
  n3 = CrossProduct(A=VertexNormalWS(), B=CrossProduct(A=Constant3Vector() {constant=<rgba 0 0 1 0>}, B=CameraVectorWS()))
  n4 = StaticSwitchParameter(
   True=MaterialFunctionCall!(
    HeightTexture=@n2.G
    TextureCoordinates=@n1
    ReferencePlane
    HeightRatioInput=ScalarParameter() {defaultValue=0.02, group=Floor_1, parameterName=BumpOffsetFloor_1}
    CameraVector=Normalize(VectorInput=@n3)
   ) {materialFunction=/Engine/Functions/Engine_MaterialFunctions02/Utility/BumpOffset_Advanced.BumpOffset_advanced}.UVs
   False=@n1
  ) {defaultValue=true, parameterName=Use_BumpOffset}
  n5 = TextureSampleParameter2D(UVs=@n4) {
   group=Floor_1
   parameterName=BaseColor_Floor_1
   samplerSource=SSM_Wrap_WorldGroupSettings
   texture=/Game/Environments/Sets/Arid/Textures/DesertSand_basecolor
  }
  n6 = ScalarParameter() {defaultValue=0.2092, group=Floor_1, parameterName=S_1_Floor_1}
  n7 = ScalarParameter() {defaultValue=0.5, group=Floor_1, parameterName=S_2_Floor_1}
  n8 = TextureSampleParameter2D(UVs=@n4) {
   group=Floor_1
   parameterName=Mask_Floor_1
   samplerSource=SSM_Wrap_WorldGroupSettings
   samplerType=SAMPLERTYPE_Masks
   texture=&DesertSand_Mask
  }
  n9 = ScalarParameter() {defaultValue=1.04, group=Floor_1, parameterName=R_Floor_1}
  n10 = TextureSampleParameter2D(UVs=@n4) {
   group=Floor_1
   parameterName=Normal_Floor_1
   samplerSource=SSM_Wrap_WorldGroupSettings
   samplerType=SAMPLERTYPE_Normal
   texture=/Game/Environments/Sets/Arid/Textures/DesertSand_normal
  }
  FunctionOutput(None=MakeMaterialAttributes!(
    BaseColor=@n5.RGB
    Metallic
    Specular=LinearInterpolate(A=@n6, B=@n7, Alpha=@n8.R)
    Roughness=Multiply(A=@n8.B, B=@n9)
    Anisotropy
    EmissiveColor
    Opacity
    OpacityMask
    Normal=@n10.RGB
    Tangent
    WorldPositionOffset
    SubsurfaceColor
    ClearCoat
    ClearCoatRoughness
    AmbientOcclusion
    Refraction
    CustomizedUVs_0
    CustomizedUVs_1
    CustomizedUVs_2
    CustomizedUVs_3
    CustomizedUVs_4
    CustomizedUVs_5
    CustomizedUVs_6
    CustomizedUVs_7
    PixelDepthOffset
    ShadingModel
    Displacement
   ))
  FunctionOutput(None=@n8.G) {outputName=Height}
 ]}
