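# Material layer function for the second wall surface in the Arid environment set, defaulting to rock cliff textures, returning material attributes plus a Height output.
# A default-on static switch applies advanced bump offset from mask green, using a camera vector rebuilt from vertex normal, world up and view cross products.
# The offset UVs drive base color, normal and mask: red blends specular, blue scales roughness, green becomes Height.
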MaterialFunction /Game/Environments/Sets/Arid/Materials/ML_Wall2
graph {nodes [
  n1 = FunctionInput(Preview=TextureCoordinate()) {bUsePreviewValueAsDefault=true, inputName=UV, inputType=FunctionInput_Vector2}
  n2 = TextureSampleParameter2D(UVs=@n1) {
   group=Wall_2
   parameterName=Mask_Wall_2
   samplerSource=SSM_Wrap_WorldGroupSettings
   samplerType=SAMPLERTYPE_Masks
   texture=/Game/Environments/Sets/Arid/Textures/RockCliff_Mask
  }
  n3 = ScalarParameter() {defaultValue=0.05, group=Wall_2, parameterName=BumpOffset_Wall2}
  BumpOffset(Coordinate=@n1, Height=@n2.G, HeightRatioInput=@n3)
  n4 = CrossProduct(A=VertexNormalWS(), B=CrossProduct(A=Constant3Vector() {constant=<rgba 0 0 1 0>}, B=CameraVectorWS()))
  n5 = StaticSwitchParameter(
   True=MaterialFunctionCall!(HeightTexture=@n2.G, TextureCoordinates=@n1, ReferencePlane, HeightRatioInput=@n3, CameraVector=Normalize(VectorInput=@n4)) {materialFunction=/Engine/Functions/Engine_MaterialFunctions02/Utility/BumpOffset_Advanced.BumpOffset_advanced}.UVs
   False=@n1
  ) {defaultValue=true, parameterName=Use_BumpOffset}
  n6 = TextureSampleParameter2D(UVs=@n5) {
   group=Wall_2
   parameterName=BaseColor_Wall_2
   samplerSource=SSM_Wrap_WorldGroupSettings
   texture=/Game/Environments/Sets/Arid/Textures/RockCliff_basecolor
  }
  n7 = ScalarParameter() {defaultValue=0.2092, group=Wall_2, parameterName=Spec_Low_Wall_2}
  n8 = ScalarParameter() {defaultValue=0.5, group=Wall_2, parameterName=Spec_High_Wall_2}
  n9 = TextureSampleParameter2D(UVs=@n5) {
   group=Wall_2
   parameterName=Mask_Wall_2
   samplerSource=SSM_Wrap_WorldGroupSettings
   samplerType=SAMPLERTYPE_Masks
   texture=&RockCliff_Mask
  }
  n10 = ScalarParameter() {defaultValue=1, group=Wall_2, parameterName=Roughness_Wall_2}
  n11 = TextureSampleParameter2D(UVs=@n5) {
   group=Wall_2
   parameterName=Normal_Wall_2
   samplerSource=SSM_Wrap_WorldGroupSettings
   samplerType=SAMPLERTYPE_Normal
   texture=/Game/Environments/Sets/Arid/Textures/RockCliff_normal
  }
  FunctionOutput(None=MakeMaterialAttributes!(
    BaseColor=@n6.RGB
    Metallic
    Specular=LinearInterpolate(A=@n7, B=@n8, Alpha=@n9.R)
    Roughness=Multiply(A=@n9.B, B=@n10)
    Anisotropy
    EmissiveColor
    Opacity
    OpacityMask
    Normal=@n11.RGB
    Tangent
    WorldPositionOffset
    SubsurfaceColor
    ClearCoat
    ClearCoatRoughness
    AmbientOcclusion
    Refraction
    CustomizedUVs_0
    CustomizedUVs_1
    CustomizedUVs_2
    CustomizedUVs_3
    CustomizedUVs_4
    CustomizedUVs_5
    CustomizedUVs_6
    CustomizedUVs_7
    PixelDepthOffset
    ShadingModel
    Displacement
   ))
  FunctionOutput(None=@n9.G) {outputName=Height, sortPriority=3}
 ]}
