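# Material function for an arid rock cliff wall layer, outputting material attributes plus a Height value.
# A switchable BumpOffset_Advanced parallax shifts UVs using mask green as height and a custom view vector from vertex normal, world up and camera vector.
# Those UVs sample base color, normal and mask, whose red blends specular, blue scales roughness and green becomes Height.
# A plain BumpOffset node is left unconnected.
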MaterialFunction /Game/Environments/Sets/Arid/Materials/ML_Wall_1
graph {nodes [
  n1 = FunctionInput(Preview=TextureCoordinate()) {bUsePreviewValueAsDefault=true, inputName=UV, inputType=FunctionInput_Vector2}
  n2 = TextureSampleParameter2D(UVs=@n1) {
   group=Wall_1
   parameterName=Mask_Wall_1
   samplerSource=SSM_Wrap_WorldGroupSettings
   samplerType=SAMPLERTYPE_Masks
   texture=/Game/Environments/Sets/Arid/Textures/RockCliffErosion_Mask
  }
  n3 = ScalarParameter() {defaultValue=0.03, group=Wall_1, parameterName=BumpOffset_Wall_1}
  BumpOffset(Coordinate=@n1, Height=@n2.G, HeightRatioInput=@n3)
  n4 = CrossProduct(A=VertexNormalWS(), B=CrossProduct(A=Constant3Vector() {constant=<rgba 0 0 1 0>}, B=CameraVectorWS()))
  n5 = StaticSwitchParameter(
   True=MaterialFunctionCall!(HeightTexture=@n2.G, TextureCoordinates=@n1, ReferencePlane, HeightRatioInput=@n3, CameraVector=Normalize(VectorInput=@n4)) {materialFunction=/Engine/Functions/Engine_MaterialFunctions02/Utility/BumpOffset_Advanced.BumpOffset_advanced}.UVs
   False=@n1
  ) {defaultValue=true, parameterName=Use_BumpOffset}
  n6 = TextureSampleParameter2D(UVs=@n5) {
   group=Wall_1
   parameterName=BaseColor_Wall_1
   samplerSource=SSM_Wrap_WorldGroupSettings
   texture=/Game/Environments/Sets/Arid/Textures/RockCliffErosion_basecolor
  }
  n7 = ScalarParameter() {defaultValue=0.4, group=Wall_1, parameterName=Spec_Low_Wall_1}
  n8 = ScalarParameter() {defaultValue=0.5, group=Wall_1, parameterName=Spec_High_Wall_1}
  n9 = TextureSampleParameter2D(UVs=@n5) {
   group=Wall_1
   parameterName=Mask_Wall_1
   samplerSource=SSM_Wrap_WorldGroupSettings
   samplerType=SAMPLERTYPE_Masks
   texture=&RockCliffErosion_Mask
  }
  n10 = ScalarParameter() {defaultValue=1, group=Wall_1, parameterName=Roughness_Wall_1}
  n11 = TextureSampleParameter2D(UVs=@n5) {
   group=Wall_1
   parameterName=Normal_Wall_1
   samplerSource=SSM_Wrap_WorldGroupSettings
   samplerType=SAMPLERTYPE_Normal
   texture=/Game/Environments/Sets/Arid/Textures/RockCliffErosion_normal
  }
  FunctionOutput(None=MakeMaterialAttributes!(
    BaseColor=@n6.RGB
    Metallic
    Specular=LinearInterpolate(A=@n7, B=@n8, Alpha=@n9.R)
    Roughness=Multiply(A=@n9.B, B=@n10)
    Anisotropy
    EmissiveColor
    Opacity
    OpacityMask
    Normal=@n11.RGB
    Tangent
    WorldPositionOffset
    SubsurfaceColor
    ClearCoat
    ClearCoatRoughness
    AmbientOcclusion
    Refraction
    CustomizedUVs_0
    CustomizedUVs_1
    CustomizedUVs_2
    CustomizedUVs_3
    CustomizedUVs_4
    CustomizedUVs_5
    CustomizedUVs_6
    CustomizedUVs_7
    PixelDepthOffset
    ShadingModel
    Displacement
   ))
  FunctionOutput(None=@n9.G) {outputName=Height, sortPriority=3}
 ]}
