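# Material function for arid environment sets that blends wall, ceiling and floor material layers by triplanar world mapping.
# Vertex normals mask the projections, while vertex color drives height blends between two wall layers and up to four floor layers.
# Floors support optional twisted UVs to distort tiling, with simpler mobile paths.
# Output gets optional color tinting and per-axis normals rebuilt into tangent space, with a switch bypassing triplanar blending.
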MaterialFunction /Game/Environments/Sets/Arid/Materials/TriplanarWorldMappedBlend
graph {nodes [
  n1 = GetMaterialAttributes(None=NamedRerouteUsage()) {attributeGetTypes [#1]}
  n2 = VectorParameter() {group=Diffuse, parameterName=Base Color Multiplier, defaultValue=<rgba 1 1 1 1>}
  n3 = VertexColor()
  n4 = StaticComponentMaskParameter(None=AppendVector(A=@n3, B=@n3)) {defaultR=true, group=Diffuse, parameterName=ModuleBaseColorChanelMask}
  n5 = StaticSwitchParameter(
   True=LinearInterpolate(
    A=VectorParameter() {group=Diffuse, parameterName=Base Color Multiplier_2, defaultValue=<rgba 1 1 1 1>}.RGB
    B=@n2.RGB
    Alpha=Saturate(None=@n4)
   )
   False=@n2.RGB
  ) {group=Diffuse, parameterName=Modulate BasewithVertexColorMask}
  FunctionOutput(None=SetMaterialAttributes!(MaterialAttributes=@n1.MaterialAttributes, BaseColor=Multiply(A=@n1.BaseColor, B=@n5)) {attributeSetTypes [#1]}) {outputName=ML_Floor_1_Only, sortPriority=3}
  n6 = Time()
  n7 = ScalarParameter() {defaultValue=1, group=Twist, parameterName=Speed}
  n8 = MaterialFunctionCall!(
   Result Bias=ScalarParameter() {defaultValue=2.4697, group=Twist, parameterName=Twist Bias}
   Result Modulation=ScalarParameter() {defaultValue=0.2752, group=Twist, parameterName=Twist Scale}
   SineAnimOffset
   SineAnimSpeed
   UVS * size=Multiply(A=ScalarParameter() {defaultValue=0.728, group=Twist, parameterName=TwirlUVSize}, B=NamedRerouteUsage())
   Rotation amount modulator=ScalarParameter() {defaultValue=0.5161, group=Twist, parameterName=RotationAmount}
   Time=StaticSwitchParameter(True=Multiply(A=@n6, B=@n7), False=ScalarParameter() {defaultValue=1, group=Twist, parameterName=UnFracTime}) {parameterName=UseAnimatedSwirl}
   Flow Direction
   Grid Noise Tile Rep=AppendVector(
    A=ScalarParameter() {defaultValue=2, group=Twist, parameterName=GridNoiseTileRep}
    B=ScalarParameter() {defaultValue=5, group=Twist, parameterName=GridNoiseTileRep2}
   )
   Twist Falloff=ScalarParameter() {defaultValue=1.25, group=Twist, parameterName=Twist Falloff Rate}
  ) {materialFunction=&/Fortnite_GroundTwistFunction}
  FunctionOutput(None=@n8.Rotation Amount) {outputName=Debug Rotation Values, sortPriority=5}
  n9 = Add(A=@n8.Distortion, B=NamedRerouteUsage())
  n10 = NamedRerouteUsage()
  n11 = NamedRerouteUsage()
  n12 = MaterialFunctionCall!(UV=ShadingPathSwitch(Default=StaticSwitchParameter(True=@n9, False=@n10) {group=Floor_1, parameterName=UseTwistUV_Floor1}, Forward=@n11, Mobile=@n11)) {materialFunction=&/ML_Floor_1}
  FunctionOutput(None=NamedRerouteDeclaration(None=@n12.Height) {name=ML_Floor_1-Height, nodeColor=<rgba 0 0.9766 1 1>}) {outputName=ML_Floor_1_Height, sortPriority=3}
  n13 = StaticSwitchParameter(
   True=MaterialFunctionCall!(Material=NamedRerouteUsage(), Tint=@n5, TintMask=Constant() {r=1}) {materialFunction=/Engine/Functions/MaterialLayerFunctions/MatLayerBlend_Tint}.Blended Material
   False=NamedRerouteUsage()
  ) {group=Diffuse, parameterName=Modulate Base Color}
  n14 = MaterialFunctionCall!(Material=NamedRerouteUsage()) {materialFunction=/Engine/Functions/MaterialLayerFunctions/MatLayerBlend_BreakNormal}
  n15 = MaterialFunctionCall(Float3=@n14.Normal) {materialFunction=&BreakOutFloat3Components}
  n16 = Multiply(A=Transform(None=Constant3Vector() {constant=<rgba 0 0 1 0>}), B=Constant() {r=1})
  n17 = AppendVector(A=Multiply(A=@n15.R) {constB=-1}, B=Multiply(A=@n15.B, B=ComponentMask(None=@n16) {g=true}))
  n18 = Multiply(A=@n15.G) {constB=-1}
  n19 = ComponentMask(None=@n16) {r=true}
  n20 = AppendVector(A=Multiply(A=@n19, B=@n15.B), B=Multiply(A=@n15.R) {constB=-1})
  n21 = Multiply(A=@n15.G) {constB=-1}
  n22 = Abs(None=@n19)
  n23 = Constant()
  n24 = Multiply(A=@n14.Normal, B=Constant3Vector() {constant=<rgba -1 -1 1 0>})
  n25 = ComponentMask(None=@n16) {b=true}
  n26 = LinearInterpolate(
   A=LinearInterpolate(
    A=AppendVector(A=@n17, B=@n18)
    B=AppendVector(A=@n20, B=@n21)
    Alpha=MaterialFunctionCall(In=@n22, Contrast=@n23) {materialFunction=&CheapContrast}.Result
   )
   B=AppendVector(A=ComponentMask(None=@n24) {g=true, r=true}, B=Multiply(A=@n15.B, B=@n25))
   Alpha=MaterialFunctionCall(In=Abs(None=@n25), Contrast=@n23) {materialFunction=&CheapContrast}.Result
  )
  n27 = MaterialFunctionCall(
   Normal=Transform(None=Normalize(VectorInput=@n26)) {transformSourceType=TRANSFORMSOURCE_World, transformType=TRANSFORM_Tangent}
   Flatness=ScalarParameter() {group=MasterControl, parameterName=NormalFlatten}
  ) {materialFunction=&FlattenNormal}
  FunctionOutput(None=StaticSwitchParameter(
    True=@n13
    False=MaterialFunctionCall!(Material=@n13, Normal=@n27.Result) {materialFunction=/Engine/Functions/MaterialLayerFunctions/MatLayerBlend_ReplaceNormals}.Result
   ) {parameterName=NoTriPlanerBlend})
  MaterialFunctionCall(
   Result Bias=ScalarParameter() {defaultValue=2.4697, group=Floor_2, parameterName=Twist Bias_2}
   Result Modulation=ScalarParameter() {defaultValue=0.2752, group=Floor_2, parameterName=Twist Scale_2}
   UVS * size=Multiply(A=ScalarParameter() {defaultValue=0.728, group=Floor_2, parameterName=TwirlUVSize_2}, B=NamedRerouteUsage())
   Rotation amount modulator=ScalarParameter() {defaultValue=0.5161, group=Floor_2, parameterName=RotationAmount_2}
   Time=ScalarParameter() {defaultValue=1, group=Floor_2, parameterName=UnFracTime_2}
   Grid Noise Tile Rep=AppendVector(
    A=ScalarParameter() {defaultValue=2, group=Floor_2, parameterName=GridNoiseTileRep_2}
    B=ScalarParameter() {defaultValue=5, group=Floor_2, parameterName=GridNoiseTileRep2_2}
   )
  ) {materialFunction=&Fortnite_GroundTwistFunction}
  NamedRerouteDeclaration(None=FunctionInput() {inputName=UVZ, inputType=FunctionInput_Vector2, sortPriority=3}) {name=UVZ, nodeColor=<rgba 0 0.9766 1 1>}
  n28 = FunctionInput() {inputName=UVX, inputType=FunctionInput_Vector2, sortPriority=1}
  n29 = MaterialFunctionCall!(UV=@n28) {materialFunction=&/ML_Wall_1}
  NamedRerouteDeclaration(None=@n29.Height) {name=UVX-Wall1-Height, nodeColor=<rgba 0 0.9766 1 1>}
  n30 = NamedRerouteUsage()
  n31 = MaterialFunctionCall!(UV=ShadingPathSwitch(Default=StaticSwitchParameter(True=@n9, False=@n10) {group=Floor_2, parameterName=UseTwistUV_Floor2}, Forward=@n30, Mobile=@n30)) {materialFunction=&/MF_Floor_2}
  NamedRerouteDeclaration(None=@n31.Result) {name=ML_Floor_2, nodeColor=<rgba 0 0.9766 1 1>}
  n32 = NamedRerouteUsage()
  n33 = MaterialFunctionCall!(UV=ShadingPathSwitch(Default=StaticSwitchParameter(True=@n9, False=@n10) {group=Floor_3, parameterName=UseTwistUV_Floor3}, Forward=@n32, Mobile=@n32)) {materialFunction=&/MF_Floor_3}
  NamedRerouteDeclaration(None=@n33.Result) {name=ML_Floor_3, nodeColor=<rgba 0 0.9766 1 1>}
  n34 = NamedRerouteUsage()
  n35 = MaterialFunctionCall!(UV=ShadingPathSwitch(Default=StaticSwitchParameter(True=@n9, False=@n10) {group=Floor_4, parameterName=UseTwistUV_Floor4}, Forward=@n34, Mobile=@n34)) {materialFunction=&/MF_Floor4}
  NamedRerouteDeclaration(None=@n35.Result) {name=ML_Floor_4, nodeColor=<rgba 0 0.9766 1 1>}
  NamedRerouteDeclaration(None=@n33.Height) {name=ML_Floor_3-Height, nodeColor=<rgba 0 0.9766 1 1>}
  NamedRerouteDeclaration(None=@n35.Height) {name=ML_Floor_4-Height, nodeColor=<rgba 0 0.9766 1 1>}
  NamedRerouteDeclaration(None=MaterialFunctionCall!(UV=NamedRerouteUsage()) {materialFunction=&/ML_Wall2}.Result) {name=ML_Wall2-Ceiling?, nodeColor=<rgba 0 0.9766 1 1>}
  n36 = VertexNormalWS()
  n37 = ComponentMask(None=@n36) {b=true}
  n38 = Multiply(A=@n37) {constB=-1}
  n39 = MaterialFunctionCall(
   In=Add(A=Clamp(None=@n38), B=ScalarParameter() {defaultValue=-0.1009, group=Blending, parameterName=TriplanarBlendContrastZOffsetTop})
   Contrast=ScalarParameter() {defaultValue=0.4603, group=Blending, parameterName=TriplanarBlendContrastZTop}
  ) {materialFunction=&CheapContrast}
  NamedRerouteDeclaration(None=Clamp(None=ComponentMask(None=@n39.Result) {r=true})) {name=TP-Ceiling, nodeColor=<rgba 0 0.9766 1 1>}
  n40 = Clamp(None=@n37)
  n41 = NamedRerouteUsage()
  n42 = MaterialFunctionCall(
   In=Add(
    A=MaterialFunctionCall(Transition Phase=@n40, Height Texture=@n41) {materialFunction=&HeightLerp}.Lerp Alpha No Contrast
    B=ScalarParameter() {defaultValue=-0.256, group=Blending, parameterName=TriplanarBlendContrastZOffset}
   )
   Contrast=ScalarParameter() {defaultValue=7.3756, group=Blending, parameterName=TriplanarBlendContrastZ}
  ) {materialFunction=&CheapContrast}
  NamedRerouteDeclaration(None=Clamp(None=ComponentMask(None=@n42.Result) {r=true})) {name=TP-Z, nodeColor=<rgba 0 0.9766 1 1>}
  n43 = Abs(None=ComponentMask(None=@n36) {r=true})
  n44 = ScalarParameter() {defaultValue=1.2134, group=Blending, parameterName=TriplanarBlendContrastXY}
  NamedRerouteDeclaration(None=ComponentMask(None=MaterialFunctionCall(In=@n43, Contrast=@n44) {materialFunction=&CheapContrast}.Result) {r=true}) {name=TP-XY, nodeColor=<rgba 0 0.9766 1 1>}
  n45 = VertexColor()
  n46 = ScalarParameter() {group=Wall_2, parameterName=VertexcolorBlendContrast_Wall2}
  n47 = MaterialFunctionCall!(
   Base Material=NamedRerouteUsage()
   Top Material=NamedRerouteUsage()
   Alpha=MaterialFunctionCall(Transition Phase=@n45.R, Height Texture=NamedRerouteUsage(), Contrast=@n46) {materialFunction=&HeightLerp}.Alpha
  ) {materialFunction=/Engine/Functions/MaterialLayerFunctions/MatLayerBlend_Standard}
  n48 = NamedRerouteUsage()
  NamedRerouteDeclaration(None=ShadingPathSwitch(
    Default=StaticSwitchParameter(True=@n47.Blended Material, False=@n48) {defaultValue=true, group=Options, parameterName=Use_Wall_2_Blend}
    Mobile=@n48
   )) {name=X-Mat, nodeColor=<rgba 0 0.9766 1 1>}
  NamedRerouteDeclaration(None=@n29.Result) {name=UVX-Wall1, nodeColor=<rgba 0 0.9766 1 1>}
  n49 = MaterialFunctionCall(
   Base Material=NamedRerouteUsage()
   Top Material=NamedRerouteUsage()
   Alpha=MaterialFunctionCall(Transition Phase=@n45.R, Height Texture=NamedRerouteUsage(), Contrast=@n46) {materialFunction=&HeightLerp}.Alpha
  ) {materialFunction=&MatLayerBlend_Standard}
  n50 = NamedRerouteUsage()
  NamedRerouteDeclaration(None=ShadingPathSwitch(
    Default=StaticSwitchParameter(True=@n49.Blended Material, False=@n50) {defaultValue=true, group=Options, parameterName=Use_Wall_2_Blend}
    Mobile=@n50
   )) {name=Y-Mat, nodeColor=<rgba 0 0.9766 1 1>}
  n51 = LinearInterpolate(
   A=ScalarParameter() {group=Floor_2, parameterName=BlendContrast_Floor_2_low}
   B=ScalarParameter() {defaultValue=1, group=Floor_2, parameterName=BlendContrast_Floor_2_High}
   Alpha=MaterialFunctionCall(
    Transition Phase=@n45.R
    Height Texture=NamedRerouteUsage()
    Contrast=ScalarParameter() {defaultValue=1, group=Floor_2, parameterName=VertexcolorBlendContrast_Floor2}
   ) {materialFunction=&HeightLerp}.Alpha
  )
  n52 = StaticSwitchParameter(
   True=MaterialFunctionCall(Base Material=NamedRerouteUsage(), Top Material=NamedRerouteUsage(), Alpha=Saturate(None=@n51)) {materialFunction=&MatLayerBlend_Standard}.Blended Material
   False=NamedRerouteUsage()
  ) {group=Options, parameterName=Use_Floor2}
  n53 = LinearInterpolate(
   A=ScalarParameter() {group=Floor_3, parameterName=BlendContrast_Floor_3_low}
   B=ScalarParameter() {defaultValue=1, group=Floor_3, parameterName=BlendContrast_Floor_3_High}
   Alpha=MaterialFunctionCall(
    Transition Phase=@n45.G
    Height Texture=NamedRerouteUsage()
    Contrast=ScalarParameter() {defaultValue=1, group=Floor_3, parameterName=VertexcolorBlendContrast_Floor3}
   ) {materialFunction=&HeightLerp}.Alpha
  )
  n54 = StaticSwitchParameter(
   True=MaterialFunctionCall(Base Material=@n52, Top Material=NamedRerouteUsage(), Alpha=Saturate(None=@n53)) {materialFunction=&MatLayerBlend_Standard}.Blended Material
   False=@n52
  ) {group=Options, parameterName=Use_Floor3}
  n55 = ScalarParameter() {group=Floor_4, parameterName=BlendContrast_Floor_4_low}
  n56 = ScalarParameter() {defaultValue=1, group=Floor_4, parameterName=BlendContrast_Floor_4_High}
  n57 = MaterialFunctionCall(
   Transition Phase=@n45.B
   Height Texture=NamedRerouteUsage()
   Contrast=ScalarParameter() {defaultValue=1, group=Floor_4, parameterName=VertexcolorBlendContrast_Floor4}
  ) {materialFunction=&HeightLerp}
  n58 = MaterialFunctionCall(Base Material=@n54, Top Material=NamedRerouteUsage(), Alpha=Saturate(None=LinearInterpolate(A=@n55, B=@n56, Alpha=@n57.Alpha))) {materialFunction=&MatLayerBlend_Standard}
  n59 = NamedRerouteUsage()
  NamedRerouteDeclaration(None=ShadingPathSwitch(Default=StaticSwitchParameter(True=@n58.Blended Material, False=@n54) {group=Options, parameterName=Use_Floor4}, Forward=@n59, Mobile=@n59)) {name=Floor-Mat, nodeColor=<rgba 0 0.9766 1 1>}
  n60 = MaterialFunctionCall(
   Base Material=MaterialFunctionCall(Base Material=NamedRerouteUsage(), Top Material=NamedRerouteUsage(), Alpha=NamedRerouteUsage()) {desc=X facing and Y facing blend, materialFunction=&MatLayerBlend_Standard}.Blended Material
   Top Material=NamedRerouteUsage()
   Alpha=NamedRerouteUsage()
  ) {desc=Blend Ceiling, materialFunction=&MatLayerBlend_Standard}
  n61 = NamedRerouteUsage()
  n62 = NamedRerouteUsage()
  NamedRerouteDeclaration(None=StaticSwitchParameter(
    True=NamedRerouteUsage()
    False=MaterialFunctionCall(Base Material=@n60.Blended Material, Top Material=@n61, Alpha=@n62) {desc=Blend Floor, materialFunction=&MatLayerBlend_Standard}.Blended Material
   ) {parameterName=NoTriPlanerBlend}) {name=BlendedMat, nodeColor=<rgba 0 0.9766 1 1>}
  NamedRerouteDeclaration(None=MaterialFunctionCall(UV=@n28) {materialFunction=&ML_Wall2}.Result) {name=UVX-Wall2, nodeColor=<rgba 0 0.9766 1 1>}
  n63 = FunctionInput() {inputName=UVY, inputType=FunctionInput_Vector2, sortPriority=2}
  NamedRerouteDeclaration(None=MaterialFunctionCall(UV=@n63) {materialFunction=&ML_Wall2}.Result) {name=UVY-Wall2, nodeColor=<rgba 0 0.9766 1 1>}
  n64 = MaterialFunctionCall(UV=@n63) {materialFunction=&ML_Wall_1}
  NamedRerouteDeclaration(None=@n64.Result) {name=UVY-Wall1, nodeColor=<rgba 0 0.9766 1 1>}
  NamedRerouteDeclaration(None=@n64.Height) {name=UVY-Wall1-Height, nodeColor=<rgba 0 0.9766 1 1>}
  NamedRerouteDeclaration(None=@n12.Result) {name=ML_Floor_1, nodeColor=<rgba 0 0.9766 1 1>}
  NamedRerouteDeclaration(None=@n31.Height) {name=ML_Floor_2-Height, nodeColor=<rgba 0 0.9766 1 1>}
 ]}
