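# WorldAlignedUVs is a material function in the Arid environment set that generates world-space planar UVs for the three axis projections.
# It divides world position, by default excluding shader offsets, by the negated absolute TextureSize input, so the coordinates are flipped and scaled to that size.
# Component masks split the result into X (GB), Y (RB) and Z (RG) outputs, each a UV pair perpendicular to its named axis.
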MaterialFunction /Game/Environments/Sets/Arid/Materials/WorldAlignedUVs
graph {nodes [
  n1 = FunctionInput() {inputName=TextureSize, previewValue=<256 256 256 1>}
  n2 = Divide(
   A=FunctionInput(Preview=WorldPosition() {worldPositionShaderOffset=WPT_ExcludeAllShaderOffsets}.XYZ) {bUsePreviewValueAsDefault=true, inputName=WorldPosition}
   B=Multiply(A=Abs(None=@n1)) {constB=-1}
  )
  FunctionOutput(None=ComponentMask(None=@n2) {b=true, g=true}) {outputName=X, sortPriority=1}
  FunctionOutput(None=ComponentMask(None=@n2) {b=true, r=true}) {outputName=Y, sortPriority=2}
  FunctionOutput(None=ComponentMask(None=@n2) {g=true, r=true}) {outputName=Z, sortPriority=3}
 ]}
