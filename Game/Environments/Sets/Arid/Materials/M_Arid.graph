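# M_Arid is the material for the Arid environment set, usable on Nanite and spline meshes, with world-aligned triplanar mapping.
# The triplanar result optionally takes a masked top layer, then a static switch picks DesertRoad or SedimentGradient processing.
# It then passes a grass switch with identical branches and an optional twist debug emissive before MF_CreativeMoveTool.
# Pixel-depth-offset and detail-normal branches also exist, but nothing shown consumes their outputs.
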Material /Game/Environments/Sets/Arid/Materials/M_Arid
graph {
 properties {
  bUseMaterialAttributes=true
  bUsedWithNanite=true
  bUsedWithSplineMeshes=true
  pixelDepthOffsetMode=PDOM_Legacy
  refractionMethod=RM_IndexOfRefraction
  usedShadingModels=Default Lit
 }
 nodes [
  n1 = ScalarParameter() {defaultValue=1000, group=MasterControl, parameterName=Texturesize_XY}
  n2 = MaterialFunctionCall!(
   WorldPosition
   TextureSize=AppendVector(A=AppendVector(A=@n1, B=@n1), B=ScalarParameter() {defaultValue=1350.7627, group=MasterControl, parameterName=Texturesize_Z})
  ) {materialFunction=&/WorldAlignedUVs}
  n3 = MaterialFunctionCall!(UVX=@n2.X, UVY=@n2.X, UVZ=@n2.X) {materialFunction=&/TriplanarWorldMappedBlend}
  n4 = MaterialFunctionCall!(UV) {materialFunction=&/MF_MaskedTopLayer}
  n5 = StaticSwitchParameter(
   True=MaterialFunctionCall!(Base Material=@n3.Result, Top Material=@n4.Result, Alpha=@n4.Result) {materialFunction=/Engine/Functions/MaterialLayerFunctions/MatLayerBlend_Standard}.Blended Material
   False=@n3.Result
  ) {parameterName=UseMaskedLayerBlend}
  n6 = StaticSwitchParameter(
   True=MaterialFunctionCall!(In=@n5) {materialFunction=&/DesertRoad}.Result
   False=MaterialFunctionCall!(In=@n5) {materialFunction=&/SedimentGradient}.Result
  ) {parameterName=IsDesertRoad}
  n7 = BreakMaterialAttributes!(Attr=@n6)
  n8 = OneMinus()
  n9 = ScalarParameter() {defaultValue=7.2072, parameterName=PDO}
  n10 = VertexNormalWS()
  n11 = CameraVectorWS()
  MakeMaterialAttributes!(
   BaseColor=@n7.BaseColor
   Metallic=@n7.BaseColor
   Specular=@n7.BaseColor
   Roughness=@n7.BaseColor
   Anisotropy
   EmissiveColor=@n7.BaseColor
   Opacity=@n7.BaseColor
   OpacityMask=@n7.BaseColor
   Normal=@n7.BaseColor
   Tangent
   WorldPositionOffset=@n7.BaseColor
   SubsurfaceColor=@n7.BaseColor
   ClearCoat=@n7.BaseColor
   ClearCoatRoughness=@n7.BaseColor
   AmbientOcclusion=@n7.BaseColor
   Refraction=@n7.BaseColor
   CustomizedUVs_0
   CustomizedUVs_1
   CustomizedUVs_2
   CustomizedUVs_3
   CustomizedUVs_4
   CustomizedUVs_5
   CustomizedUVs_6
   CustomizedUVs_7
   PixelDepthOffset=Divide(A=Multiply(A=@n8, B=@n9), B=DotProduct(A=@n10, B=@n11))
   ShadingModel
   Displacement
  )
  n12 = StaticSwitchParameter(True=@n6, False=@n6) {defaultValue=true, group=Layers, parameterName="Use Grass(true) or a BaseTilingTexture(False)"}
  MaterialFunctionCall(MA In=StaticSwitchParameter(
    True=SetMaterialAttributes(MaterialAttributes=@n12, Emissive Color=@n3.Debug Rotation Values) {attributeSetTypes [#1]}
    False=@n12
   ) {group=Twist, parameterName=Debug Twist Amount}) {materialFunction=&MF_CreativeMoveTool}
  n13 = TextureCoordinate()
  n14 = ScalarParameter() {defaultValue=1.0108, parameterName=DetailNormal}
  MaterialFunctionCall!(
   Material=@n6
   Additional Normal=TextureSample(UVs=Multiply(A=@n13, B=@n14)) {samplerType=SAMPLERTYPE_Normal, texture=/Game/Environments/Sets/Arid/Textures/RockDetail_normal}.RGB
   Mask
  ) {materialFunction=/Engine/Functions/MaterialLayerFunctions/MatLayerBlend_BlendAngleCorrectedNormals}
 ]
}
