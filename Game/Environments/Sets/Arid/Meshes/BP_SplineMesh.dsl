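; Spline mesh actor Blueprint for environment sets that lays a static mesh along an editable spline in its construction script.
; The built-in component spans the first segment, and an added component per further segment copies its mesh, forward axis, shadow and lighting channels.
; Optional settings apply an override material and spline-point scaling.
; Auto-smoothing rescales tangents on added segments to point distance times Tension; event graph stubs are empty.
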(event EventBeginPlay)

(event Collision|EventActorBeginOverlap
  (params (OtherActor :type Actor%20Object%20Reference)))

(event EventTick
  (params (DeltaSeconds :type Float%20%28single-precision%29)))

(fn ConstructionScript
  (bind (_bchannel0 _bchannel1 _bchannel2) (Utilities|Struct|BreakLightingChannels (Variables|/Script/Engine.PrimitiveComponent|GetLightingChannels (Variables|SplineMeshActor|GetSplineMeshComponent))))
  (bind _spline (Variables|Default|GetSpline))
  (bind (_location _tangent) (Spline|GetLocationandTangentatSplinePoint _spline 1 "Local"))
  (bind (_location_2 _tangent_2) (Spline|GetLocationandTangentatSplinePoint _spline 0 "Local"))
  (bind _splinemeshcomponent (Variables|SplineMeshActor|GetSplineMeshComponent))
  (bind _splinemeshcomponent_2 (Variables|SplineMeshActor|GetSplineMeshComponent))
  (bind _spline_2 (Variables|Default|GetSpline))
  (bind _overridematerial (Variables|Default|GetOverrideMaterial))
  (bind _overridematerial_2 (Variables|Default|GetOverrideMaterial))
  (SplineMesh|SetStartandEnd _splinemeshcomponent _location_2 _tangent_2 _location _tangent true :!class /Script/Engine.SplineMeshComponent)
  (Utilities|FlowControl|Sequence
    (:then_0
      (if (Utilities|IsValid _overridematerial_2 :!class /Script/Engine.KismetSystemLibrary)
        (Rendering|Material|SetMaterial _splinemeshcomponent 0 _overridematerial_2)))
    (:then_1
      (if (Variables|Default|GetUseScaling)
        (SplineMesh|SetStartScale _splinemeshcomponent (Math|Vector2D|MakeVector2D (.y (Spline|GetScaleatSplinePoint _spline 0)) (.z (Spline|GetScaleatSplinePoint _spline 0))) true :!class /Script/Engine.SplineMeshComponent)
        (SplineMesh|SetEndScale _splinemeshcomponent (Math|Vector2D|MakeVector2D (.y (Spline|GetScaleatSplinePoint _spline 1)) (.z (Spline|GetScaleatSplinePoint _spline 1))) true :!class /Script/Engine.SplineMeshComponent))
      (for _index (range 1 (+ (- (Spline|GetNumberOfSplinePoints _spline_2) 2) 1))
        (bind _returnvalue (AddComponent|NODE_AddSplineMeshComponent-0))
        (Rendering|Components|SetLightingChannels _returnvalue _bchannel0 _bchannel1 _bchannel2)
        (Components|StaticMesh|SetStaticMesh _returnvalue (Variables|/Script/Engine.StaticMeshComponent|GetStaticMesh _splinemeshcomponent_2))
        (Utilities|FlowControl|Sequence
          (:then_0
            (if (Utilities|IsValid _overridematerial :!class /Script/Engine.KismetSystemLibrary)
              (Rendering|Material|SetMaterial _returnvalue 0 _overridematerial)))
          (:then_1
            (SplineMesh|SetForwardAxis _returnvalue (SplineMesh|GetForwardAxis _splinemeshcomponent) true)
            (Rendering|SetCastShadow _returnvalue (Variables|/Script/Engine.PrimitiveComponent|GetCastShadow _splinemeshcomponent_2))
            (SplineMesh|SetStartandEnd _returnvalue (Spline|GetLocationandTangentatSplinePoint _spline_2 _index "Local") (Math|Vector|SelectVector (* (Math|Vector|Normalize (.Tangent (Spline|GetLocationandTangentatSplinePoint _spline_2 _index "Local"))) (* (Math|Vector|VectorLength (- (Spline|GetLocationandTangentatSplinePoint _spline_2 (+ _index 1) "Local") (Spline|GetLocationandTangentatSplinePoint _spline_2 _index "Local"))) (Variables|Default|GetTension))) (.Tangent (Spline|GetLocationandTangentatSplinePoint _spline_2 _index "Local")) (Variables|Default|GetAutoSmoothTangent)) (Spline|GetLocationandTangentatSplinePoint _spline_2 (+ _index 1) "Local") (Math|Vector|SelectVector (* (Math|Vector|Normalize (.Tangent (Spline|GetLocationandTangentatSplinePoint _spline_2 (+ _index 1) "Local"))) (* (Math|Vector|VectorLength (- (Spline|GetLocationandTangentatSplinePoint _spline_2 (+ _index 1) "Local") (Spline|GetLocationandTangentatSplinePoint _spline_2 _index "Local"))) (Variables|Default|GetTension))) (.Tangent (Spline|GetLocationandTangentatSplinePoint _spline_2 (+ _index 1) "Local")) (Variables|Default|GetAutoSmoothTangent)) true :!class /Script/Engine.SplineMeshComponent)
            (if (Variables|Default|GetUseScaling)
              (SplineMesh|SetStartScale _returnvalue (Math|Vector2D|MakeVector2D (.y (Spline|GetScaleatSplinePoint _spline_2 _index)) (.z (Spline|GetScaleatSplinePoint _spline_2 _index))) true :!class /Script/Engine.SplineMeshComponent)
              (SplineMesh|SetEndScale _returnvalue (Math|Vector2D|MakeVector2D (.y (Spline|GetScaleatSplinePoint _spline_2 (+ _index 1))) (.z (Spline|GetScaleatSplinePoint _spline_2 (+ _index 1)))) true :!class /Script/Engine.SplineMeshComponent))))))))
